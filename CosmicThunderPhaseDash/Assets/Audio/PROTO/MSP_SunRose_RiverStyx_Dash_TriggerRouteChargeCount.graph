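# MetaSound patch that routes an incoming trigger to one of four trigger outputs by ChargeCount, apparently for the SunRose RiverStyx dash sound.
# A chain of integer Trigger Compare nodes tests the count against 0 through 3, each False output feeding the next check and each True firing its output.
# The first and last checks use different comparison types from the middle two, apparently so out-of-range counts land on the end outputs.
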MetaSoundPatch /CosmicThunderPhaseDash/Assets/Audio/PROTO/MSP_SunRose_RiverStyx_Dash_TriggerRouteChargeCount
inputs [
 ChargeCount {type=Int32}
 Input {type=Trigger}
]
outputs [
 0 {type=Trigger}
 Trigger1 {type=Trigger}
 Trigger2 {type=Trigger}
 Trigger3 {type=Trigger}
]
pages [Default {
  input_defaults {ChargeCount=0, Input=null}
  nodes [
   TriggerCompareClamp_3 = TriggerCompare.Clamp:Int32(A=^ChargeCount, Compare=^Input) {also {defaults {B=0, Type=4}}}
   TriggerCompareClamp_1 = TriggerCompare.Clamp:Int32(A=^ChargeCount, Compare=@TriggerCompareClamp_3.False) {also {defaults {B=1}}}
   TriggerCompareClamp_0 = TriggerCompare.Clamp:Int32(A=^ChargeCount, Compare=@TriggerCompareClamp_1.False) {also {defaults {B=2}}}
   TriggerCompareClamp_2 = TriggerCompare.Clamp:Int32(A=^ChargeCount, Compare=@TriggerCompareClamp_0.False) {also {defaults {B=3, Type=5}}}
  ]
  links [
   {dst="@Trigger2", src=TriggerCompareClamp_0.True}
   {dst="@Trigger1", src=TriggerCompareClamp_1.True}
   {dst="@Trigger3", src=TriggerCompareClamp_2.True}
   {dst="@0", src=TriggerCompareClamp_3.True}
  ]
 }]
