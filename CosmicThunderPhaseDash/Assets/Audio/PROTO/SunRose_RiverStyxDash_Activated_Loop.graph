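# Mono looping MetaSound for the activated state of the SunRose River Styx Dash, layering a base loop with randomized one-shots.
# One-shot pools follow the charge count input; each finished sound schedules the next after a random delay, with varied pitch and gain.
# Output fades in, then eases to half level after a delay.
# A Mobile page drops charge branching and plays one pool of LayerTwo sounds.
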MetaSoundSource /CosmicThunderPhaseDash/Assets/Audio/PROTO/SunRose_RiverStyxDash_Activated_Loop
inputs [
 BaseLoop.Sound {type=WaveAsset}
 BaseLoop.Volume {type=Float}
 FadeInTime {type=Time}
 FromGCN.ChargeCount {type=Int32}
 OneShots.OneCharge {type=WaveAsset:Array}
 OneShots.ThreeCharges {type=WaveAsset:Array}
 OneShots.TwoCharges {type=WaveAsset:Array}
 OneShots.Volume {type=Float}
 OneShots.ZeroCharges {type=WaveAsset:Array}
 UE.Source.OnPlay {type=Trigger}
]
interfaces [UE.OutputFormat.Mono:1.0, UE.Source:1.1]
outputs ["UE.OutputFormat.Mono.Audio:0" {type=Audio}]
pages [
 Default {
  is_build_graph=true
  input_defaults {
   BaseLoop.Sound=&/SunRose_RiverStyx_Activated_Base_Loop_01
   BaseLoop.Volume=0.2
   FadeInTime=2
   FromGCN.ChargeCount=3
   OneShots.Volume=0.9
   UE.Source.OnPlay=null
   OneShots.OneCharge [
    &/SunRose_RiverStyx_LayerOne_01
    &/SunRose_RiverStyx_LayerOne_02
    &/SunRose_RiverStyx_LayerOne_03
    &/SunRose_RiverStyx_LayerOne_04
    &/SunRose_RiverStyx_LayerOne_05
   ]
   OneShots.ThreeCharges [
    &/SunRose_RiverStyx_LayerThree_01
    &/SunRose_RiverStyx_LayerThree_02
    &/SunRose_RiverStyx_LayerThree_03
    &/SunRose_RiverStyx_LayerThree_04
    &/SunRose_RiverStyx_LayerThree_05
    &/SunRose_RiverStyx_LayerThree_06
   ]
   OneShots.TwoCharges [
    &/SunRose_RiverStyx_LayerTwo_06
    &/SunRose_RiverStyx_LayerTwo_05
    &/SunRose_RiverStyx_LayerTwo_04
    &/SunRose_RiverStyx_LayerTwo_03
    &/SunRose_RiverStyx_LayerTwo_02
    &/SunRose_RiverStyx_LayerTwo_01
   ]
   OneShots.ZeroCharges [
    &/SunRose_RiverStyx_Activated_OneShots_01
    &/SunRose_RiverStyx_Activated_OneShots_02
    &/SunRose_RiverStyx_Activated_OneShots_03
    &/SunRose_RiverStyx_Activated_OneShots_04
    &/SunRose_RiverStyx_Activated_OneShots_05
    &/SunRose_RiverStyx_Activated_OneShots_06
    &/SunRose_RiverStyx_Activated_OneShots_07
    &/SunRose_RiverStyx_Activated_OneShots_08
    &/SunRose_RiverStyx_Activated_OneShots_09
    &/SunRose_RiverStyx_Activated_OneShots_10
    &/SunRose_RiverStyx_Activated_OneShots_11
    &/SunRose_RiverStyx_Activated_OneShots_12
   ]
  }
  nodes [
   n1 = InitVariable.Trigger()
   n2 = VariableDeferredAccessor.Trigger(Variable=InitVariable.Trigger().Variable)
   TriggerAnyTriggerAny2_0 = "TriggerAny.Trigger Any (2)"(In 0=^UE.Source.OnPlay, In 1=@n2.Value)
   n3 = UE.RandomTime(Next=@TriggerAnyTriggerAny2_0.Out) {also {defaults {Max=2, Min=0.3}}}
   TriggerDelay_0 = UE.Trigger Delay(In=@n3.On Next, Delay Time=@n3.Value)
   n4 = TriggerCompare.Clamp:Int32(
    A=^FromGCN.ChargeCount
    Compare=VariableAccessor.Trigger(Variable=VariableMutator.Trigger(Variable=@n1.Variable, Value=@TriggerDelay_0.Out).Variable).Value
   ) {also {defaults {B=0, Type=4}}}
   n5 = TriggerCompare.Clamp:Int32(A=^FromGCN.ChargeCount, Compare=@n4.False) {also {defaults {B=1}}}
   n6 = TriggerCompare.Clamp:Int32(A=^FromGCN.ChargeCount, Compare=@n5.False) {also {defaults {B=2}}}
   ArrayRandomGet_1 = Array.Random Get:WaveAsset:Array(In Array=^OneShots.TwoCharges, Next=@n6.True) {also {defaults {No Repeats=5, Weights []}}}
   n7 = UE.RandomFloat(Next=@ArrayRandomGet_1.On Next) {also {defaults {Max=3, Min=-3}}}
   WavePlayer_3 = UE.Wave Player:Mono(Wave Asset=@ArrayRandomGet_1.Value, Play=@n7.On Next, Pitch Shift=@n7.Value) {also {version [1,1]}}
   ArrayRandomGet_0 = Array.Random Get:WaveAsset:Array(In Array=^OneShots.ZeroCharges, Next=@n4.True) {also {defaults {No Repeats=5, Weights []}}}
   n8 = UE.RandomFloat(Next=@ArrayRandomGet_0.On Next) {also {defaults {Max=5, Min=-5}}}
   WavePlayer_1 = UE.Wave Player:Mono(Wave Asset=@ArrayRandomGet_0.Value, Play=@n8.On Next, Pitch Shift=@n8.Value) {also {version [1,1]}}
   ArrayRandomGet_2 = Array.Random Get:WaveAsset:Array(
    In Array=^OneShots.ThreeCharges
    Next=TriggerCompare.Clamp:Int32(A=^FromGCN.ChargeCount, Compare=@n6.False) {also {defaults {B=3, Type=5}}}.True
   ) {also {defaults {No Repeats=5, Weights []}}}
   n9 = UE.RandomFloat(Next=@ArrayRandomGet_2.On Next) {also {defaults {Max=3, Min=-3}}}
   WavePlayer_0 = UE.Wave Player:Mono(Wave Asset=@ArrayRandomGet_2.Value, Play=@n9.On Next, Pitch Shift=@n9.Value) {also {version [1,1]}}
   ArrayRandomGet_3 = Array.Random Get:WaveAsset:Array(In Array=^OneShots.OneCharge, Next=@n5.True) {also {defaults {No Repeats=5, Weights []}}}
   n10 = UE.RandomFloat(Next=@ArrayRandomGet_3.On Next) {also {defaults {Max=3, Min=-3}}}
   WavePlayer_2 = UE.Wave Player:Mono(Wave Asset=@ArrayRandomGet_3.Value, Play=@n10.On Next, Pitch Shift=@n10.Value) {also {version [1,1]}}
   AudioMixerAudioMixerMono4_0 = "AudioMixer.Audio Mixer (Mono, 4)"(
    Gain 2=UE.InterpTo:Audio(Target=UE.RandomFloat(Next=@WavePlayer_3.On Play) {also {defaults {Max=0.6, Min=0.4}}}.Value) {also {defaults {Interp Time=0.2}}}.Value
    Gain 0=UE.InterpTo:Audio(Target=UE.RandomFloat(Next=@WavePlayer_1.On Play) {also {defaults {Max=0.8, Min=0.5}}}.Value) {also {defaults {Interp Time=0.2}}}.Value
    Gain 3=UE.InterpTo:Audio(Target=UE.RandomFloat(Next=@WavePlayer_0.On Play) {also {defaults {Max=0.6, Min=0.4}}}.Value) {also {defaults {Interp Time=0.2}}}.Value
    Gain 1=UE.InterpTo:Audio(Target=UE.RandomFloat(Next=@WavePlayer_2.On Play) {also {defaults {Max=0.6, Min=0.4}}}.Value) {also {defaults {Interp Time=0.2}}}.Value
    In 3=@WavePlayer_0.Out Mono
    In 0=@WavePlayer_1.Out Mono
    In 1=@WavePlayer_2.Out Mono
    In 2=@WavePlayer_3.Out Mono
   )
   WavePlayer_4 = UE.Wave Player:Mono(Wave Asset=^BaseLoop.Sound, Play=^UE.Source.OnPlay) {also {defaults {Loop="<MetaSoundBuilderResult.SUCCEEDED: 0>"}, version [1,1]}}
   AudioMixerAudioMixerMono2_0 = "AudioMixer.Audio Mixer (Mono, 2)"(Gain 0=^BaseLoop.Volume, Gain 1=^OneShots.Volume, In 1=@AudioMixerAudioMixerMono4_0.Out, In 0=@WavePlayer_4.Out Mono)
   TriggerDelay_1 = UE.Trigger Delay(In=^UE.Source.OnPlay) {also {defaults {Delay Time=5}}}
   n11 = Value.Value:Float(Set=@TriggerDelay_1.Out, Reset=VariableAccessor.Trigger(Variable=InitVariable.Trigger().Variable).Value) {also {defaults {Init Value=1, Target Value=0.5}}}
   n12 = Value.Value:Float(Set=^UE.Source.OnPlay) {also {defaults {Init Value=0, Target Value=1}}}
   Multiply_0 = UE.Multiply:Audio by Float(
    PrimaryOperand=@AudioMixerAudioMixerMono2_0.Out
    AdditionalOperands=UE.Multiply:Float(
     PrimaryOperand=UE.InterpTo:Audio(Target=@n11.Output Value) {also {defaults {Interp Time=5}}}.Value
     AdditionalOperands=UE.InterpTo:Audio(Interp Time=^FadeInTime, Target=@n12.Output Value).Value
    ).Out
   )
   TriggerAnyTriggerAny4_0 = "TriggerAny.Trigger Any (4)"(In 3=@WavePlayer_0.On Finished, In 0=@WavePlayer_1.On Finished, In 1=@WavePlayer_2.On Finished, In 2=@WavePlayer_3.On Finished)
   VariableMutator.Trigger(Value=@TriggerAnyTriggerAny4_0.Out, Variable=@n2.Variable)
  ]
  links [{dst="@UE.OutputFormat.Mono.Audio:0", src=Multiply_0.Out}]
 }
 Mobile {
  input_defaults {
   OneShots.OneCharge []
   OneShots.ThreeCharges []
   OneShots.TwoCharges []
   OneShots.ZeroCharges [
    &SunRose_RiverStyx_LayerTwo_06
    &SunRose_RiverStyx_LayerTwo_05
    &SunRose_RiverStyx_LayerTwo_04
    &SunRose_RiverStyx_LayerTwo_03
    &SunRose_RiverStyx_LayerTwo_02
    &SunRose_RiverStyx_LayerTwo_01
   ]
  }
  nodes [
   n1 = InitVariable.Trigger()
   n2 = VariableDeferredAccessor.Trigger(Variable=InitVariable.Trigger().Variable)
   TriggerAnyTriggerAny2_0 = "TriggerAny.Trigger Any (2)"(In 0=^UE.Source.OnPlay, In 1=@n2.Value)
   n3 = UE.RandomTime(Next=@TriggerAnyTriggerAny2_0.Out) {also {defaults {Max=2, Min=0.3}}}
   TriggerDelay_0 = UE.Trigger Delay(In=@n3.On Next, Delay Time=@n3.Value)
   ArrayRandomGet_0 = Array.Random Get:WaveAsset:Array(
    In Array=^OneShots.ZeroCharges
    Next=VariableAccessor.Trigger(Variable=VariableMutator.Trigger(Variable=@n1.Variable, Value=@TriggerDelay_0.Out).Variable).Value
   ) {also {defaults {No Repeats=5, Weights []}}}
   n4 = UE.RandomFloat(Next=@ArrayRandomGet_0.On Next) {also {defaults {Max=5, Min=-5}}}
   WavePlayer_0 = UE.Wave Player:Mono(Wave Asset=@ArrayRandomGet_0.Value, Play=@n4.On Next, Pitch Shift=@n4.Value) {also {version [1,1]}}
   n5 = UE.RandomFloat(Next=@WavePlayer_0.On Play) {also {defaults {Max=0.8, Min=0.5}}}
   WavePlayer_1 = UE.Wave Player:Mono(Wave Asset=^BaseLoop.Sound, Play=^UE.Source.OnPlay) {also {defaults {Loop="<MetaSoundBuilderResult.SUCCEEDED: 0>"}, version [1,1]}}
   AudioMixerAudioMixerMono2_0 = "AudioMixer.Audio Mixer (Mono, 2)"(
    Gain 0=^BaseLoop.Volume
    Gain 1=UE.Multiply:Float(AdditionalOperands=^OneShots.Volume, PrimaryOperand=UE.InterpTo:Audio(Target=@n5.Value) {also {defaults {Interp Time=0.2}}}.Value).Out
    In 1=@WavePlayer_0.Out Mono
    In 0=@WavePlayer_1.Out Mono
   )
   TriggerDelay_1 = UE.Trigger Delay(In=^UE.Source.OnPlay) {also {defaults {Delay Time=5}}}
   n6 = Value.Value:Float(Set=@TriggerDelay_1.Out, Reset=VariableAccessor.Trigger(Variable=InitVariable.Trigger().Variable).Value) {also {defaults {Init Value=1, Target Value=0.5}}}
   n7 = Value.Value:Float(Set=^UE.Source.OnPlay) {also {defaults {Init Value=0, Target Value=1}}}
   Multiply_0 = UE.Multiply:Audio by Float(
    PrimaryOperand=@AudioMixerAudioMixerMono2_0.Out
    AdditionalOperands=UE.Multiply:Float(
     PrimaryOperand=UE.InterpTo:Audio(Target=@n6.Output Value) {also {defaults {Interp Time=5}}}.Value
     AdditionalOperands=UE.InterpTo:Audio(Interp Time=^FadeInTime, Target=@n7.Output Value).Value
    ).Out
   )
   VariableMutator.Trigger(Variable=@n2.Variable, Value=@WavePlayer_0.On Finished)
  ]
  links [{dst="@UE.OutputFormat.Mono.Audio:0", src=Multiply_0.Out}]
 }
]
properties {
 assetImportData=&self:AssetImportData
 attenuationSettings=/CosmicThunderPhaseDash/Assets/Audio/PROTO/Settings/SunRose_RiverStyx_Activate
 bLooping=true
 bOverrideConcurrency=true
 duration=10000
 maxDistance=1800
 sampleRate=48000
 soundClassObject=&Fort_Gadgets
 volume=0.4
 concurrencyOverrides {maxCount=3}
}
