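# One-shot stereo MetaSound source for the SunRose RiverStyx deactivation sound, in the Fort_Gadgets sound class.
# On play, it picks one of three deactivate waves with repeat avoidance, then rolls a random downward pitch shift for the stereo wave player.
# Playback completion fires OnFinished.
# It reuses the attenuation settings named for the Activate sound.
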MetaSoundSource /CosmicThunderPhaseDash/Assets/Audio/PROTO/SunRose_RiverStyx_Deactivate
inputs [
 Deactivate Sounds {type=WaveAsset:Array}
 UE.Source.OnPlay {type=Trigger}
]
interfaces [UE.OutputFormat.Stereo:1.0, UE.Source.OneShot:1.0, UE.Source:1.1]
outputs [
 "UE.OutputFormat.Stereo.Audio:0" {type=Audio}
 "UE.OutputFormat.Stereo.Audio:1" {type=Audio}
 UE.Source.OneShot.OnFinished {type=Trigger}
]
pages [Default {
  input_defaults {
   UE.Source.OnPlay=null
   Deactivate Sounds [
    &/SunRose_RiverStyx_Deactivate_01
    &/SunRose_RiverStyx_Deactivate_02
    &/SunRose_RiverStyx_Deactivate_03
   ]
  }
  nodes [
   ArrayRandomGet_0 = Array.Random Get:WaveAsset:Array(In Array=^Deactivate Sounds, Next=^UE.Source.OnPlay) {also {defaults {No Repeats=3}}}
   n1 = UE.RandomFloat(Next=@ArrayRandomGet_0.On Next) {also {defaults {Max=-3, Min=-1}}}
   WavePlayer_0 = UE.Wave Player:Stereo(Wave Asset=@ArrayRandomGet_0.Value, Play=@n1.On Next, Pitch Shift=@n1.Value)
  ]
  links [
   {dst="@UE.Source.OneShot.OnFinished", src=WavePlayer_0.On Finished}
   {dst="@UE.OutputFormat.Stereo.Audio:0", src=WavePlayer_0.Out Left}
   {dst="@UE.OutputFormat.Stereo.Audio:1", src=WavePlayer_0.Out Right}
  ]
 }]
properties {
 assetImportData=&self:AssetImportData
 attenuationSettings=/CosmicThunderPhaseDash/Assets/Audio/PROTO/Settings/SunRose_RiverStyx_Activate
 maxDistance=1800
 numChannels=2
 outputFormat=Stereo
 sampleRate=48000
 soundClassObject=&Fort_Gadgets
 volume=0.7
}
